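# Physics asset for the dynamic collar and necklace on the medium female Elf Attack player body.
# Capsules form the neck and upper spine, while boxes shape the collar panels and necklace.
# Side and back panels swing within limits from tiny MSTR anchor bodies, which join spine_05 with free angular motion.
# Front panels attach directly to spine_05 with tighter limits, and the necklace hangs from it as a three-link, low-twist chain.
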PhysicsAsset /BRCosmetics/Characters/Player/Female/Medium/Bodies/F_MED_ElfAttack/Meshes/F_MED_Elf_Attack_Collar_Dyn
bodies {
 dyn_collar_bk [{shapeType=Box, extentX=9.0691, extentY=3.6922, extentZ=13.237, shapeName=back, center=<6.1977 -2.8729 0.0001>, rotation=<rot 0 -22.2595 0>}]
 dyn_collar_bk_MSTR [{shapeType=Capsule, length=0.505, radius=0.505, rotation=<rot 0 -90 0>}]
 dyn_collar_bk_l [{
   shapeType=Box
   extentX=9.0691
   extentY=4.5
   extentZ=10.3312
   shapeName=back
   center=<6.3285 -0.8163 -2.4981>
   rotation=<rot -26.568 8.7789 -86.8701>
  }]
 dyn_collar_bk_l_MSTR [{shapeType=Capsule, length=0.505, radius=0.505, rotation=<rot 0 -90 0>}]
 dyn_collar_bk_r [{
   shapeType=Box
   extentX=7.0593
   extentY=5.0495
   extentZ=10.3312
   shapeName=back
   center=<-7.0905 0.7232 3.3147>
   rotation=<rot -28.03 8.3139 -88.2178>
  }]
 dyn_collar_bk_r_MSTR [{shapeType=Capsule, length=0.505, radius=0.505, rotation=<rot 0 -90 0>}]
 dyn_collar_fr [{shapeType=Box, extentX=10.5, extentY=4.5, extentZ=3, center=<5.0969 0 0>, rotation=<rot 0 23.7853 0>}]
 dyn_collar_fr_l [{shapeType=Box, extentX=2.6014, extentY=4.5, extentZ=1.6284, center=<2.6799 0 -0.3395>}]
 dyn_collar_fr_r [{shapeType=Box, extentX=2.5301, extentY=4.5, extentZ=1.5973, center=<-2.5236 -0.0001 0.3254>}]
 dyn_collar_l [
  {
   shapeType=Box
   extentX=8.8985
   extentY=13.9896
   extentZ=3
   shapeName=fr
   center=<4.2548 -0.0423 -2.5515>
   rotation=<rot -26.3802 1.4315 5.1115>
  }
  {
   shapeType=Box
   extentX=9.53
   extentY=13.9896
   extentZ=3
   shapeName=bk
   center=<4.1394 -5.5351 -0.3389>
   rotation=<rot -25.3963 -26.4126 38.6552>
  }
 ]
 dyn_collar_l_MSTR [{shapeType=Capsule, length=0.505, radius=0.505, rotation=<rot 0 -90 0>}]
 dyn_collar_r [
  {
   shapeType=Box
   extentX=13.295
   extentY=13.9896
   extentZ=3
   shapeName=fr
   center=<-5.6816 0.0321 3.3118>
   rotation=<rot -25.9059 17.019 -1.681>
  }
  {
   shapeType=Box
   extentX=9.53
   extentY=11.6238
   extentZ=3
   shapeName=bk
   center=<-4.0152 5.7664 1.1486>
   rotation=<rot -21.3923 -16.2876 31.8013>
  }
 ]
 dyn_collar_r_MSTR [{shapeType=Capsule, length=0.505, radius=0.505, rotation=<rot 0 -90 0>}]
 dyn_necklace_01 [{shapeType=Box, extentX=6.875, extentY=0.75, extentZ=5, center=<-3.5832 0 0.0001>}]
 dyn_necklace_02 [{shapeType=Box, extentX=2.625, extentY=0.75, extentZ=4.0625, center=<-1.0491 0 0.0002>}]
 dyn_necklace_03 [{shapeType=Box, extentX=3.75, extentY=0.75, extentZ=3.5625, center=<-1.8616 0 0.0001>}]
 neck_02 [
  {shapeType=Capsule, length=2.8586, radius=8.319, center=<10.2262 -2.7272 0.0231>, rotation=<rot 0.4639 -100.7558 89.9063>}
  {shapeType=Capsule, length=2.9693, radius=7.891, center=<6.2508 0.8412 -2.5807>, rotation=<rot 2.2789 -153.808 73.8624>}
  {shapeType=Capsule, length=6.8163, radius=5.5717, center=<6.8487 -0.707 3.6399>, rotation=<rot 3.5413 -162.0271 -81.4798>}
  {shapeType=Capsule, length=2.564, radius=6.3847, center=<10.2027 -4.5442 0.0371>, rotation=<rot 0.4639 -100.7564 -0.0933>}
 ]
 spine_03 [{shapeType=Capsule, length=8.5407, radius=8.5708, center=<4.8005 5.1262 -0.1037>, rotation=<rot 1.1877 -90 -0.4746>}]
 spine_04 [{shapeType=Capsule, length=7.5719, radius=9.5396, center=<-0.7534 3.7775 -0.122>, rotation=<rot 1.1877 -90 -0.4746>}]
 spine_05 [{shapeType=Capsule, length=9.9375, radius=9.5625, center=<-6.3632 2.8293 0>}]
}
constraints [
 {bone1Name=dyn_necklace_01, bone2Name=spine_05, swing1LimitDegrees=45, swing2LimitDegrees=45, twistLimitDegrees=1}
 {bone1Name=dyn_necklace_02, bone2Name=dyn_necklace_01, swing1LimitDegrees=45, swing2LimitDegrees=45, twistLimitDegrees=1}
 {bone1Name=dyn_necklace_03, bone2Name=dyn_necklace_02, swing1LimitDegrees=45, swing2LimitDegrees=45, twistLimitDegrees=1}
 {bone1Name=dyn_collar_fr_r, bone2Name=spine_05, swing1LimitDegrees=25, swing2LimitDegrees=25, twistLimitDegrees=5}
 {bone1Name=dyn_collar_fr_l, bone2Name=spine_05, swing1LimitDegrees=25, swing2LimitDegrees=25, twistLimitDegrees=5}
 {bone1Name=dyn_collar_fr, bone2Name=spine_05, swing1LimitDegrees=25, swing2LimitDegrees=25, twistLimitDegrees=5}
 {bone1Name=dyn_collar_r, bone2Name=dyn_collar_r_MSTR, swing1LimitDegrees=60, swing2LimitDegrees=60, twistLimitDegrees=30}
 {bone1Name=dyn_collar_l, bone2Name=dyn_collar_l_MSTR, swing1LimitDegrees=60, swing2LimitDegrees=60, twistLimitDegrees=30}
 {bone1Name=dyn_collar_bk, bone2Name=dyn_collar_bk_MSTR, swing1LimitDegrees=90, swing2LimitDegrees=90, twistLimitDegrees=30}
 {
  bone1Name=spine_05
  bone2Name=spine_04
  swing1LimitDegrees=45
  swing1Motion=Free
  swing2LimitDegrees=45
  swing2Motion=Free
  twistLimitDegrees=45
  twistMotion=Free
 }
 {
  bone1Name=spine_04
  bone2Name=spine_03
  swing1LimitDegrees=45
  swing1Motion=Free
  swing2LimitDegrees=45
  swing2Motion=Free
  twistLimitDegrees=45
  twistMotion=Free
 }
 {
  bone1Name=neck_02
  bone2Name=spine_05
  swing1LimitDegrees=45
  swing1Motion=Free
  swing2LimitDegrees=45
  swing2Motion=Free
  twistLimitDegrees=45
  twistMotion=Free
 }
 {
  bone1Name=dyn_collar_bk_l_MSTR
  bone2Name=spine_05
  swing1LimitDegrees=45
  swing1Motion=Free
  swing2LimitDegrees=45
  swing2Motion=Free
  twistLimitDegrees=45
  twistMotion=Free
 }
 {bone1Name=dyn_collar_bk_l, bone2Name=dyn_collar_bk_l_MSTR, swing1LimitDegrees=90, swing2LimitDegrees=90, twistLimitDegrees=30}
 {
  bone1Name=dyn_collar_bk_MSTR
  bone2Name=spine_05
  swing1LimitDegrees=45
  swing1Motion=Free
  swing2LimitDegrees=45
  swing2Motion=Free
  twistLimitDegrees=45
  twistMotion=Free
 }
 {
  bone1Name=dyn_collar_bk_r_MSTR
  bone2Name=spine_05
  swing1LimitDegrees=45
  swing1Motion=Free
  swing2LimitDegrees=45
  swing2Motion=Free
  twistLimitDegrees=45
  twistMotion=Free
 }
 {bone1Name=dyn_collar_bk_r, bone2Name=dyn_collar_bk_r_MSTR, swing1LimitDegrees=90, swing2LimitDegrees=90, twistLimitDegrees=30}
 {
  bone1Name=dyn_collar_r_MSTR
  bone2Name=spine_05
  swing1LimitDegrees=45
  swing1Motion=Free
  swing2LimitDegrees=45
  swing2Motion=Free
  twistLimitDegrees=45
  twistMotion=Free
 }
 {
  bone1Name=dyn_collar_l_MSTR
  bone2Name=spine_05
  swing1LimitDegrees=45
  swing1Motion=Free
  swing2LimitDegrees=45
  swing2Motion=Free
  twistLimitDegrees=45
  twistMotion=Free
 }
]
