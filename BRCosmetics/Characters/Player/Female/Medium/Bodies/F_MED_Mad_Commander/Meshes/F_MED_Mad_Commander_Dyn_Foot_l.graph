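# Physics asset for the left boot of the F_MED_Mad_Commander character body, covering the boot, its inner and outer top flaps, and two straps.
# Each bone carries one capsule, and constraints chain them up to calf_twist_01_l.
# The main boot is held almost rigid to the calf, while both top flaps and the bottom strap hang from it with moderate swing and twist limits.
# The top strap hangs from the outer flap.
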PhysicsAsset /BRCosmetics/Characters/Player/Female/Medium/Bodies/F_MED_Mad_Commander/Meshes/F_MED_Mad_Commander_Dyn_Foot_l
bodies {
 calf_twist_01_l [{shapeType=Capsule, length=1.5, radius=0.5, rotation=<rot -90 0.0396 -0.0396>}]
 dyn_boot_l [{shapeType=Capsule, length=3, radius=3, center=<-2.7 0 0>, rotation=<rot 90 0 0>}]
 dyn_boot_top_in_l [{shapeType=Capsule, length=5, radius=1.5, center=<-4 0 0>, rotation=<rot -90 0 0>}]
 dyn_boot_top_out_l [{shapeType=Capsule, length=5, radius=1.5, center=<-4 0 0>, rotation=<rot -90 0 0>}]
 dyn_bootstrap_bott_l [{shapeType=Capsule, length=1, radius=1, center=<0 0 2>}]
 dyn_bootstrap_top_l [{shapeType=Capsule, length=1, radius=1, center=<0 0 2>}]
}
constraints [
 {bone1Name=dyn_bootstrap_top_l, bone2Name=dyn_boot_top_out_l, swing1LimitDegrees=25, swing2LimitDegrees=25, twistLimitDegrees=20}
 {bone1Name=dyn_boot_top_out_l, bone2Name=dyn_boot_l, swing1LimitDegrees=20, swing2LimitDegrees=20, twistLimitDegrees=10}
 {bone1Name=dyn_bootstrap_bott_l, bone2Name=dyn_boot_l, swing1LimitDegrees=25, swing2LimitDegrees=25, twistLimitDegrees=20}
 {bone1Name=dyn_boot_top_in_l, bone2Name=dyn_boot_l, swing1LimitDegrees=20, swing2LimitDegrees=20, twistLimitDegrees=10}
 {bone1Name=dyn_boot_l, bone2Name=calf_twist_01_l, swing1LimitDegrees=1, swing2LimitDegrees=1, twistLimitDegrees=0}
]
