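# Physics asset for the dynamic right boot of the female medium Mad Commander cosmetic body, covering the boot, its top flaps and straps.
# Capsule bodies form a small constraint tree rooted at the calf twist bone, with the main boot almost locked to it.
# Inner and outer boot tops and the bottom strap swing within limited cones off the boot.
# The top strap chains from the outer top.
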PhysicsAsset /BRCosmetics/Characters/Player/Female/Medium/Bodies/F_MED_Mad_Commander/Meshes/F_MED_Mad_Commander_Dyn_Foot_r
bodies {
 calf_twist_01_r [{shapeType=Capsule, length=1.5, radius=0.5, rotation=<rot -90 0.0396 -0.0396>}]
 dyn_boot_r [{shapeType=Capsule, length=3, radius=3, center=<2.7 0 0>, rotation=<rot 90 -0.0198 -0.0198>}]
 dyn_boot_top_in_r [{shapeType=Capsule, length=5, radius=1.5, center=<4 0 0>, rotation=<rot 90 0 0>}]
 dyn_boot_top_out_r [{shapeType=Capsule, length=5, radius=1.5, center=<4 0 0>, rotation=<rot 90 0 0>}]
 dyn_bootstrap_bott_r [{shapeType=Capsule, length=1, radius=1, center=<0 0 -2>}]
 dyn_bootstrap_top_r [{shapeType=Capsule, length=1, radius=1, center=<0 0 -2>}]
}
constraints [
 {bone1Name=dyn_boot_top_out_r, bone2Name=dyn_boot_r, swing1LimitDegrees=20, swing2LimitDegrees=20, twistLimitDegrees=10}
 {bone1Name=dyn_bootstrap_top_r, bone2Name=dyn_boot_top_out_r, swing1LimitDegrees=25, swing2LimitDegrees=25, twistLimitDegrees=20}
 {bone1Name=dyn_bootstrap_bott_r, bone2Name=dyn_boot_r, swing1LimitDegrees=25, swing2LimitDegrees=25, twistLimitDegrees=20}
 {bone1Name=dyn_boot_top_in_r, bone2Name=dyn_boot_r, swing1LimitDegrees=20, swing2LimitDegrees=20, twistLimitDegrees=10}
 {bone1Name=dyn_boot_r, bone2Name=calf_twist_01_r, swing1LimitDegrees=1, swing2LimitDegrees=1, twistLimitDegrees=0}
]
