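# Post-process style cosmetic AnimBP for the F_MED_ElfAttack female body on the Mad Commander skeleton, correcting collar bones and adding secondary physics.
# It copies the attached parent's pose, then applies additive collar offsets: constant at front, others blended by head yaw relative to spine_05.
# Crouch events drive slowly interpolated extra offsets on right-side collar bones.
# Three RigidBody nodes then simulate both feet and the collar in calf and spine_05 space.
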AnimBlueprint /BRCosmetics/Characters/Player/Female/Medium/Bodies/F_MED_ElfAttack/Meshes/F_MED_ElfAttack_AnimBP
target_skeleton=/BRCosmetics/Characters/Player/Female/Medium/Bodies/F_MED_Mad_Commander/Meshes/F_MED_Mad_Commander_Skeleton
graphs [
 AnimGraph {nodes [
   n1 = "Transform (Modify) Bone - Bone: dyn_collar_fr"!(
    ComponentPose=Local To Component(LocalPose=Copy Pose From Mesh() {bUseAttachedParent=true})
    Alpha=1
    Translation="0,0,0"
    Rotation="0,-20,0"
   ) -> (Pose) {
    rotationMode=BMM_Additive
    rotationSpace=BCS_BoneSpace
    translationMode=BMM_Additive
    translationSpace=BCS_BoneSpace
    alphaBoolBlend {blendOption=Linear}
    alphaScaleBiasClamp {interpSpeedDecreasing=20, interpSpeedIncreasing=20}
    boneToModify {boneName=dyn_collar_fr}
    rotation {yaw=-20}
   }
   n2 = Get head up() -> (head up:double)
   n3 = "Transform (Modify) Bone - Bone: dyn_collar_bk_MSTR"!(ComponentPose=@n1, Alpha=@n2, Translation="0,0,0", Rotation="0.000,-15,0.000") -> (Pose) {
    rotationMode=BMM_Additive
    rotationSpace=BCS_BoneSpace
    translationMode=BMM_Additive
    translationSpace=BCS_BoneSpace
    alphaBoolBlend {blendOption=Linear}
    alphaScaleBiasClamp {bInterpResult=true, interpSpeedDecreasing=20, interpSpeedIncreasing=20}
    boneToModify {boneName=dyn_collar_bk_MSTR}
    rotation {yaw=-15}
   }
   n4 = "Transform (Modify) Bone - Bone: dyn_collar_l_MSTR"!(ComponentPose=@n3, Alpha=@n2, Translation="0,0,0", Rotation="12,-10,0") -> (Pose) {
    rotationMode=BMM_Additive
    rotationSpace=BCS_BoneSpace
    translationMode=BMM_Additive
    translationSpace=BCS_BoneSpace
    alphaBoolBlend {blendOption=Linear}
    alphaScaleBiasClamp {bInterpResult=true, interpSpeedDecreasing=20, interpSpeedIncreasing=20}
    boneToModify {boneName=dyn_collar_l_MSTR}
    rotation {pitch=12, yaw=-10}
   }
   n5 = "Transform (Modify) Bone - Bone: dyn_collar_bk_l_MSTR"!(ComponentPose=@n4, Alpha=@n2, Translation="0,0,0", Rotation="5,-10,10") -> (Pose) {
    rotationMode=BMM_Additive
    rotationSpace=BCS_BoneSpace
    translationMode=BMM_Additive
    translationSpace=BCS_BoneSpace
    alphaBoolBlend {blendOption=Linear}
    alphaScaleBiasClamp {bInterpResult=true, interpSpeedDecreasing=20, interpSpeedIncreasing=20}
    boneToModify {boneName=dyn_collar_bk_l_MSTR}
    rotation=<rot 5 -10 10>
   }
   n6 = "Transform (Modify) Bone - Bone: dyn_collar_bk_r_MSTR"!(ComponentPose=@n5, Alpha=@n2, Translation="0,0,0", Rotation="5,0.000,0.000") -> (Pose) {
    rotationMode=BMM_Additive
    rotationSpace=BCS_BoneSpace
    translationMode=BMM_Additive
    translationSpace=BCS_BoneSpace
    alphaBoolBlend {blendOption=Linear}
    alphaScaleBiasClamp {bInterpResult=true, interpSpeedDecreasing=20, interpSpeedIncreasing=20}
    boneToModify {boneName=dyn_collar_bk_r_MSTR}
    rotation {pitch=5}
   }
   n7 = "Transform (Modify) Bone - Bone: dyn_collar_r_MSTR"!(ComponentPose=@n6, Alpha=@n2, Translation="0,0,0", Rotation="15,0,-10") -> (Pose) {
    rotationMode=BMM_Additive
    rotationSpace=BCS_BoneSpace
    translationMode=BMM_Additive
    translationSpace=BCS_BoneSpace
    alphaBoolBlend {blendOption=Linear}
    alphaScaleBiasClamp {bInterpResult=true, interpSpeedDecreasing=20, interpSpeedIncreasing=20}
    boneToModify {boneName=dyn_collar_r_MSTR}
    rotation {pitch=15, roll=-10}
   }
   n8 = Get crouch alpha() -> (crouch alpha:double)
   n9 = "Transform (Modify) Bone - Bone: dyn_collar_bk_r_MSTR"(ComponentPose=@n7, Alpha=@n8, Translation="2.482,2.328,0.646", Rotation="-9.312,0.000,0.000") {
    rotationMode=BMM_Additive
    rotationSpace=BCS_BoneSpace
    translationMode=BMM_Additive
    translationSpace=BCS_BoneSpace
    alphaBoolBlend {blendOption=Linear}
    alphaScaleBiasClamp {bInterpResult=true, interpSpeedDecreasing=0.2, interpSpeedIncreasing=0.2}
    boneToModify {boneName=dyn_collar_bk_r_MSTR}
    rotation {pitch=-9.312}
    translation=<2.482 2.328 0.646>
   }
   n10 = "Transform (Modify) Bone - Bone: dyn_collar_r_MSTR"(ComponentPose=@n9, Alpha=@n8, Translation="0.953,0.000,0.305", Rotation="-14.109,0.000,0.000") {
    rotationMode=BMM_Additive
    rotationSpace=BCS_BoneSpace
    translationMode=BMM_Additive
    translationSpace=BCS_BoneSpace
    alphaBoolBlend {blendOption=Linear}
    alphaScaleBiasClamp {bInterpResult=true, interpSpeedDecreasing=0.2, interpSpeedIncreasing=0.2}
    boneToModify {boneName=dyn_collar_r_MSTR}
    rotation {pitch=-14.109}
    translation {x=0.953, z=0.305}
   }
   n11 = RigidBody(ComponentPose=@n10) {
    bOverrideWorldGravity=true
    overridePhysicsAsset=/BRCosmetics/Characters/Player/Female/Medium/Bodies/F_MED_Mad_Commander/Meshes/F_MED_Mad_Commander_Dyn_Foot_l
    simulationSpace=BaseBoneSpace
    alphaBoolBlend {blendOption=Linear}
    baseBoneRef {boneName=calf_l}
    componentAppliedLinearAccClamp=<1000 1000 1000>
    componentLinearAccScale=<5 5 5>
    componentLinearVelScale=<5 5 5>
    overrideWorldGravity {z=-980}
   }
   n12 = RigidBody(ComponentPose=@n11) {
    bOverrideWorldGravity=true
    overridePhysicsAsset=/BRCosmetics/Characters/Player/Female/Medium/Bodies/F_MED_Mad_Commander/Meshes/F_MED_Mad_Commander_Dyn_Foot_r
    simulationSpace=BaseBoneSpace
    alphaBoolBlend {blendOption=Linear}
    baseBoneRef {boneName=calf_r}
    componentAppliedLinearAccClamp=<1000 1000 1000>
    componentLinearAccScale=<5 5 5>
    componentLinearVelScale=<5 5 5>
    overrideWorldGravity {z=-980}
   }
   n13 = RigidBody(ComponentPose=@n12) {
    bOverrideWorldGravity=true
    overridePhysicsAsset=&/F_MED_Elf_Attack_Collar_Dyn
    simulationSpace=BaseBoneSpace
    alphaBoolBlend {blendOption=Linear}
    baseBoneRef {boneName=spine_05}
    componentAppliedLinearAccClamp=<1000 1000 1000>
    componentLinearAccScale=<2 1 1>
    componentLinearVelScale=<2 1 1>
    overrideWorldGravity {z=-980}
   }
   Output Pose(Result=Component To Local(ComponentPose=@n13)) {layerGroup=None}
  ]}
 EventGraph {nodes [
   Event BlueprintUpdateAnimation()
   n1 = BreakTransform(InTransform=GetDeltaTransformFromRefPose(self=GetOwnerSkeletalMeshComponent(), BoneName=head, BaseName=spine_05))
   > Set head up(head up:double=MapRangeClamped(Value=BreakRotator(InRot=@n1.Rotation).Yaw, InRangeA=-20, InRangeB=-40)) -> (Output_Get:double)
   Event OnBeginCrouching()
   > Set crouch alpha(crouch alpha:double=1) -> (Output_Get:double)
   Event OnEndCrouching()
   > Set crouch alpha(crouch alpha=0)
  ]}
]
variables [
 crouch alpha {type=number}
 head up {type=number}
]
