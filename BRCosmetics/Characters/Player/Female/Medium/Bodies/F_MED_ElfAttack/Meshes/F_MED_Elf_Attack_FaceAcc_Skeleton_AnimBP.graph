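# Animation Blueprint for the face accessory mesh of the female medium Elf Attack body, following the parent mesh's pose and adding hair motion.
# Each update maps the head's yaw relative to spine_05 from its ref pose into a doubled, clamped headUp value.
# headUp drives the alpha of a bone-space Copy Bone from head to head_offset, covering translation and rotation.
# Anim Dynamics then simulates dyn_hair_01 with a damped angular spring and tight angular limits.
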AnimBlueprint /BRCosmetics/Characters/Player/Female/Medium/Bodies/F_MED_ElfAttack/Meshes/F_MED_Elf_Attack_FaceAcc_Skeleton_AnimBP
target_skeleton=&/F_MED_Elf_Attack_FaceAcc_Skeleton
graphs [
 AnimGraph {nodes [
   n1 = Copy Bone - Source Bone: head - Target Bone: head_offset!(
    ComponentPose:Component Space Pose Link Structure=Local To Component(LocalPose=Copy Pose From Mesh() {bUseAttachedParent=true})
    Alpha:float=Get headUp() -> (headUp:double)
    bCopyTranslation:bool=true
    bCopyRotation:bool=true
    bCopyScale:bool=False
   ) -> (Pose:Component Space Pose Link Structure) {
    is=title AnimGraphNode_CopyBone
    controlSpace=BCS_BoneSpace
    alphaBoolBlend {blendOption=Linear}
    alphaScaleBiasClamp {inRange {max=2}}
    sourceBone {boneName=head}
    targetBone {boneName=head_offset}
   }
   n2 = Anim Dynamics - Bone: dyn_hair_01(ComponentPose=@n1, Alpha=1) {
    angularDampingOverride=0.99
    angularSpringConstant=300
    bAngularSpring=true
    bGravityOverrideInSimSpace=true
    bOverrideAngularDamping=true
    bUseGravityOverride=true
    simulationSpace=BoneRelative
    alphaBoolBlend {blendOption=Linear}
    boundBone {boneName=dyn_hair_01}
    componentAppliedLinearAccClamp=<10000 10000 10000>
    componentLinearAccScale=<0.3 0.3 0.3>
    gravityOverride {z=-980}
    physicsBodyDefinitions [{
      sphereCollisionRadius=0
      boundBone {boneName=dyn_hair_01}
      boxExtents=<5 5 5>
      constraintSetup {angularLimitsMax=<2 2 7>, angularLimitsMin=<-2 -2 -7>, angularTarget=<0 0 0>}
      localJointOffset=<1 0 0>
     }]
    relativeSpaceBone {boneName=spine_05}
   }
   Output Pose(Result=Component To Local(ComponentPose=@n2)) {layerGroup=None}
  ]}
 EventGraph {nodes [
   Event BlueprintUpdateAnimation()
   n1 = GetOwnerSkeletalMeshComponent()
   n2 = BreakRotator(InRot=BreakTransform(InTransform=GetDeltaTransformFromRefPose(self=@n1, BoneName=head, BaseName=spine_05)).Rotation)
   > Set headUp(headUp:double=float * float(A=MapRangeClamped(Value=@n2.Yaw, InRangeA=18, InRangeB=-40), B=2)) -> (Output_Get:double)
   TryGetPawnOwner()
  ]}
]
variables [headUp {type=number}]
